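# Temporary beacon screen material, apparently for the reboot machine, showing a battery charge display or a broken-state repair icon.
# State fades from a dim screen to the blinking battery, cut off by Charge, then to the blinking repair icon and fail color.
# Panning scanlines and timed UV band glitches add distortion; vertex green mirrors U.
# Sphere mask and Fresnel dim edges, time-of-day scales brightness, and low quality drops scanlines, glitch and edge dimming.
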Material /Game/Effects/Fort_Effects/Materials/Mash/M_Beacon_Screen_TEMP
graph {
 properties {pixelDepthOffsetMode=PDOM_Legacy, usedShadingModels=Default Lit}
 nodes [
  Constant() {r=0.5}
  Constant() {r=0.1}
  n1 = Constant() {r=1}
  n2 = ScalarParameter() {defaultValue=1, parameterName=HeightScale}
  n3 = Panner(Coordinate=Multiply(
    A=TextureCoordinate() {vTiling=32}
    B=MaterialFunctionCall(X=@n1, Y=@n2) {materialFunction=&MakeFloat2}.Result
   )) {speedX=0.5, speedY=0.7}
  n4 = QualitySwitch(
   Default=Multiply(A=TextureSample(UVs=@n3) {texture=/Game/Missions/Primary/LaunchTheBalloon/Textures/T_TV_Scanlines}.RGB) {constB=0.05}
   Low=Constant()
  )
  n5 = ScalarParameter() {defaultValue=1, parameterName=State}
  n6 = Clamp(None=Subtract(A=Clamp(None=@n5) {maxDefault=2}))
  n7 = Constant() {r=0.1}
  n8 = Multiply(
   A=Time()
   B=LinearInterpolate(
    A=ScalarParameter() {defaultValue=1, parameterName=Charging Battery Blink Intensity}
    B=ScalarParameter() {defaultValue=2, parameterName=Broken Blink Intensity}
    Alpha=@n6
   )
  )
  n9 = Floor(None=ConstantBiasScale(None=Sine(None=@n8)) {bias=2})
  n10 = Frac(None=ComponentMask(None=TextureCoordinate() {vTiling=2.2}) {g=true})
  n11 = Multiply(A=Time()) {constB=0.23}
  n12 = Floor(None=Multiply(A=Frac(None=@n11)) {constB=1.1})
  n13 = Multiply(A=Floor(None=Multiply(A=@n10) {constB=3}), B=@n12) {constA=0.3}
  n14 = Clamp(None=Multiply(B=Multiply(A=@n13, B=@n12)) {constA=-0.03}) {minDefault=-1}
  n15 = Constant()
  n16 = Add(A=QualitySwitch(Default=AppendVector(A=@n14, B=@n15), Low=Constant()), B=TextureCoordinate())
  n17 = ComponentMask(None=VectorParameter() {parameterName=TextureScale, defaultValue=<rgba 1.5 1.125 0 0>}.RGB) {g=true, r=true}
  n18 = VectorParameter() {parameterName=TextureOffset, defaultValue=<rgba -0.25 -0.12 0 0>}
  n19 = MaterialFunctionCall(Float2=Add(A=Multiply(A=@n16, B=@n17), B=ComponentMask(None=@n18.RGB) {g=true, r=true})) {materialFunction=&BreakOutFloat2Components}
  n20 = AppendVector(A=LinearInterpolate(A=@n19.R, B=OneMinus(None=@n19.R), Alpha=VertexColor().G), B=@n19.G)
  n21 = TextureSample(UVs=@n20) {texture=/Game/Athena/Items/EnvironmentalItems/SCMachine/FX/T_Reboot_Battery_FX}
  n22 = Multiply(A=@n9, B=@n21.R)
  n23 = Add(
   A=ComponentMask(None=@n20) {r=true}
   B=LinearInterpolate(
    A=ScalarParameter() {parameterName=BatteryGradMin}
    B=ScalarParameter() {defaultValue=-1, parameterName=BatteryGradMax}
    Alpha=ScalarParameter() {defaultValue=0.5, parameterName=Charge}
   ) {constB=6}
  )
  n24 = OneMinus(None=Clamp(None=Ceil(None=@n23)))
  n25 = TextureSample(UVs=@n20) {samplerSource=SSM_Clamp_WorldGroupSettings, texture=/Game/Effects/Fort_Effects/Textures/T_Repair_Icon_BR}
  n26 = LinearInterpolate(A=Add(A=Multiply(A=@n22, B=@n24), B=@n21.G), B=Multiply(A=@n9, B=Multiply(A=@n25.R, B=@n25.R)), Alpha=@n6)
  n27 = Multiply(
   A=LinearInterpolate(
    A=VectorParameter() {parameterName=ScreenColor, defaultValue=<rgba 0.1205 0.7309 1.5 0>}.RGB
    B=VectorParameter() {parameterName=ScreenColorFail, defaultValue=<rgba 1.5 0.1677 0.108 0>}.RGB
    Alpha=@n6
   )
   B=LinearInterpolate(A=@n7, B=Add(A=@n26, B=@n7), Alpha=Clamp(None=@n5))
  ) {constB=0.3}
  n28 = ScalarParameter() {defaultValue=0.5, parameterName=FrasnelDarkenExp}
  n29 = Multiply(
   A=SphereMask(A=TextureCoordinate(), B=Constant() {r=0.5}, Radius=ScalarParameter() {defaultValue=0.74, parameterName=EdgeDimRadius}) {hardnessPercent=0}
   B=OneMinus(None=Fresnel(ExponentIn=@n28) {baseReflectFraction=0, exponent=0.5})
  )
  n30 = Constant() {r=1}
  n31 = CollectionParameter() {
   collection=&FortniteMaterialParameters
   parameterName=FortniteActiveTimeOfDay
  }
  Multiply(
   A=Multiply(A=Add(A=@n4, B=@n27), B=QualitySwitch(Default=@n29, Low=@n30))
   B=LinearInterpolate(B=ScalarParameter() {defaultValue=0.55, parameterName=NightBrightness}, Alpha=ComponentMask(None=@n31) {a=true}) {constA=1}
  )
 ]
}
