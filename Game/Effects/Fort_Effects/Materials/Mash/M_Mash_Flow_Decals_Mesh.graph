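# Unlit translucent particle material for the Mash flow decal effect, usable on mesh and sprite particles.
# Four panning shield-caustics samples form a particle-tinted flow, added to a hue-shifted, Fresnel-driven thin-film smoke lookup and remapped to HDR.
# Particle alpha, a UV sphere mask and a squared fade on the distance between actor and scene behind translucency shape a mask.
# Blended sine waves with a per-client offset parameter flicker that mask.
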Material /Game/Effects/Fort_Effects/Materials/Mash/M_Mash_Flow_Decals_Mesh
graph {
 properties {
  bEnableResponsiveAA=true
  bUsedWithMeshParticles=true
  bUsedWithParticleSprites=true
  blendMode=BLEND_Translucent
  pixelDepthOffsetMode=PDOM_Legacy
  previewMesh=/Game/Effects/Fort_Effects/Gameplay/Mash/SM_Mash_Beacon_Flow
  shadingModel=MSM_Unlit
  usedShadingModels=Unlit
 }
 nodes [
  n1 = VertexNormalWS()
  n2 = Multiply(
   A=ScalarParameter() {defaultValue=0.25, group=ThinFilm, parameterName=ThinFilm_Scale}
   B=Fresnel(ExponentIn=ScalarParameter() {defaultValue=1, group=ThinFilm, parameterName=ThinFilm_EX}, Normal=@n1) {baseReflectFraction=0}
  )
  n3 = Multiply(
   A=TextureSample(UVs=Panner(Coordinate=@n2) {speedX=-0.1}) {texture=/Game/Global/Textures/Default/Tileable_Patterns/T_Smoke_Puff_Tile_Up}.RGB
   B=ScalarParameter() {defaultValue=2, group=ThinFilm, parameterName=ThinFilm_Intensity}
  )
  n4 = Fresnel(ExponentIn=ScalarParameter() {defaultValue=1.2164, group=ThinFilm, parameterName=ThinFilm_Visibility}, Normal=@n1) {baseReflectFraction=0}
  n5 = MaterialFunctionCall(
   Hue Shift Percentage=ScalarParameter() {defaultValue=0.65, parameterName="Hue Shift Percentage (S)"}
   Texture=LinearInterpolate(A=@n3, B=@n3, Alpha=@n4)
  ) {materialFunction=&HueShift}
  n6 = TextureCoordinate() {vTiling=2}
  n7 = TextureCoordinate() {uTiling=2, vTiling=3}
  n8 = LinearInterpolate(
   A=TextureSample(UVs=Panner(Coordinate=@n6) {speedX=-0.5, speedY=0.25}) {texture=&T_ShieldCaustics}.G
   B=TextureSample(UVs=Panner(Coordinate=@n7) {speedX=1, speedY=0.5}) {texture=&T_ShieldCaustics}.G
  )
  n9 = TextureCoordinate()
  n10 = TextureCoordinate()
  n11 = LinearInterpolate(
   A=TextureSample(UVs=Panner(Coordinate=@n9) {speedX=-0.5, speedY=0.5}) {texture=&T_ShieldCaustics}.G
   B=TextureSample(UVs=Panner(Coordinate=@n10) {speedX=1, speedY=0.15}) {texture=&T_ShieldCaustics}.G
  )
  n12 = ParticleColor()
  n13 = Multiply(A=Power(Base=LinearInterpolate(A=@n8, B=@n11), Exp=ScalarParameter() {defaultValue=1.25, parameterName=FlameAmount}), B=@n12.RGB)
  n14 = MaterialFunctionCall(LDR Input=Clamp(None=Add(A=@n5.Result, B=@n13)), HDR Tint=Multiply(A=Fresnel(), B=@n12.RGB)) {materialFunction=&DeriveHDRfromLDR}
  Add(A=@n5.Result, B=@n14.RemappedRange)
  n15 = TextureCoordinate()
  n16 = Constant() {r=0.5}
  n17 = Divide(
   A=Distance(
    A=MaterialFunctionCall() {materialFunction=&WorldPositionBehindTranslucency}.Result
    B=ActorPositionWS()
   )
   B=ScalarParameter() {defaultValue=25, group=Distance Fade, parameterName=Fade Distance}
  )
  n18 = ScalarParameter() {defaultValue=25, group=Distance Fade, parameterName=Fade Start Distance, sortPriority=1}
  n19 = OneMinus(None=Clamp(None=Subtract(A=@n17, B=@n18)))
  n20 = Multiply(A=Multiply(A=Multiply(A=@n13, B=@n12.A), B=SphereMask(A=@n15, B=@n16, Radius=@n16) {hardnessPercent=0}), B=Multiply(A=@n19, B=@n19))
  Clamp(None=@n20)
  MaterialFunctionCall(Emissive=@n14.RemappedRange) {materialFunction=&MF_FX_EmissiveAdjustments}
  n21 = ParticlePositionWS()
  n22 = CameraPositionWS()
  n23 = ScalarParameter()
  Multiply(A=Normalize(VectorInput=Subtract(A=@n21, B=@n22)), B=Multiply(A=ComponentMask(None=@n23) {r=true}) {constB=-0.5}) {desc=WPO input}
  n24 = LinearInterpolate(Alpha=ScalarParameter() {defaultValue=1, parameterName=RandomOffsetPerClient_Mesh})
  n25 = Add(A=Time() {bOverride_Period=true, period=2.6}, B=@n24)
  n26 = Add(A=Time() {bOverride_Period=true, period=1.57}, B=@n24)
  n27 = Constant() {r=0.15}
  n28 = Constant() {r=0.85}
  n29 = Sine(None=Add(A=Time() {bOverride_Period=true, period=1.24}, B=@n24)) {period=1.24}
  Multiply(
   A=@n20
   B=LinearInterpolate(
    A=Sine(None=@n25) {period=2.6}
    B=Sine(None=@n26) {period=1.57}
    Alpha=LinearInterpolate(A=@n27, B=@n28, Alpha=@n29) {constA=0.15, constB=0.85}
   )
  )
  n30 = OneMinus(None=ScalarParameter() {defaultValue=0.7, parameterName=Roughness})
  n31 = ScalarParameter() {defaultValue=1, group=ThinFilm, parameterName=ThinFilm_RoughnessScale}
  Multiply(A=Clamp(None=Multiply(A=@n30, B=@n31)), B=@n4)
  Multiply()
  ParticleDirection()
  ParticleSize()
 ]
}
