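# M_Mash_Runes is an additive, unlit particle material for glowing purple rune sprites.
# Vertex colour encodes each sprite's offset around the component pivot, and the Sprite function places and scales it there.
# A time loop offset by the RuneOffset dynamic parameter drives a pulse shaping scale, brightness and opacity.
# A 2x2 SubUV mask cycles glyphs, blending dim to bright colour, with particle colour, camera fade and depth fade applied.
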Material /Game/Effects/Fort_Effects/Materials/Mash/M_Mash_Runes
graph {
 properties {
  bUsedWithParticleSprites=true
  blendMode=BLEND_Additive
  pixelDepthOffsetMode=PDOM_Legacy
  shadingModel=MSM_Unlit
  twoSided=true
  usedShadingModels=Unlit
 }
 nodes [
  n1 = LinearInterpolate(B=Constant3Vector() {constant=<rgba 1 1 1 0>}, Alpha=VertexColor()) {constA=-1}
  n2 = ScalarParameter() {defaultValue=359.2359, parameterName=Position}
  n3 = DynamicParameter() {paramNames [RuneOffset,Param2,Param3,Param4]}
  n4 = ScalarParameter() {defaultValue=64, parameterName=Offset_1}
  n5 = Time()
  n6 = ScalarParameter() {defaultValue=0.1, parameterName=ScaleSpeed}
  n7 = Frac(None=Add(A=Multiply(A=@n3.RuneOffset, B=@n4), B=Multiply(A=@n5, B=@n6)))
  n8 = Multiply(A=Multiply(A=OneMinus(None=@n7), B=SquareRoot(None=@n7))) {constB=4}
  MaterialFunctionCall!(
   Center Location=Add(
    A=Multiply(A=@n1, B=@n2)
    B=MaterialFunctionCall() {materialFunction=&ComponentPivotLocation}.Result
   )
   0-1 UVs
   XY scale=Multiply(A=ScalarParameter() {defaultValue=120.4234, parameterName=Scale}, B=@n8)
   OverridePivot
   "Normals (Optional)"
  ) {materialFunction=&Sprite}
  n9 = Multiply(A=DynamicParameter() {paramNames [RuneOffset,Param2,Param3,Param4]}.RuneOffset, B=ScalarParameter() {defaultValue=64, parameterName=Offset}) {constB=2}
  n10 = Multiply(A=Time(), B=ScalarParameter() {defaultValue=1, parameterName=Speed})
  n11 = MaterialFunctionCall(Float3=MaterialFunctionCall!(
    Texture=TextureObject() {samplerType=SAMPLERTYPE_LinearColor, texture=/Game/Athena/Prototype/Blueprints/Cube/Materials/DFHeightMask}
    UVs=TextureCoordinate()
    SubImages=Constant2Vector() {g=2, r=2}
    Frame=Add(A=@n9, B=@n10)
   ) {materialFunction=/Engine/Functions/Engine_MaterialFunctions03/Texturing/SubUV_Function}.RGB) {materialFunction=&BreakOutFloat3Components}
  n12 = LinearInterpolate(A=@n11.R, B=@n11.R, Alpha=@n8)
  n13 = LinearInterpolate(
   A=VectorParameter() {parameterName=RuneColorDim, defaultValue=<rgba 0.1718 0.0748 0.8802 0>}.RGB
   B=VectorParameter() {parameterName=RuneColorBright, defaultValue=<rgba 1.378 0.1161 2 0>}.RGB
   Alpha=@n12
  )
  MaterialFunctionCall(Emissive=Multiply(A=ParticleColor().RGB, B=Multiply(A=@n8, B=@n13))) {materialFunction=&MF_FX_EmissiveAdjustments}
  n14 = LinearInterpolate(Alpha=@n8) {constA=1, constB=4}
  n15 = Multiply(
   A=Multiply(A=Power(Base=@n12, Exp=@n14), B=Power(Base=@n8) {constExponent=4})
   B=ScalarParameter() {defaultValue=0.0847, parameterName=Opacity}
  ) {constB=0.2}
  n16 = MaterialFunctionCall() {materialFunction=&CameraDepthFade}
  Multiply(
   A=ParticleColor().A
   B=DepthFade(Opacity=Multiply(A=@n15, B=@n16.Result), FadeDistance=ScalarParameter() {defaultValue=16, parameterName=DepthFade})
  )
  VertexColor()
  VertexColor()
 ]
}
